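# Conversation database for the ReptilianOcean NPC, defining its greeting and the top-level branches of its dialogue.
# The entry point leads to one Speech node whose contextual lines react to a first-ever meeting, specific player outfits or the NPC's match conversation count, with a default fallback.
# The Speech node fans out through link nodes to shared quest and basic service conversations, the NPC's own Dialogue entry, and three Season 29 Snapshot quest entries.
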ConversationDatabase /NPCLibrary/NPCs/ReptilianOcean/NPCLibrary_Conversation_ReptilianOcean
entry_points [{tag=Conversation.Entry.ReptilianOcean, destinations [#1]}]
nodes [
 ConversationEntryPointNode_7 {
  id=#1
  class=ConversationEntryPointNode
  outputs [#2]
  properties {entryTag {tagName=Conversation.Entry.ReptilianOcean}}
 }
 Speech_C_4 {
  id=#2
  class=Speech_C
  outputs [
   #3
   #4
   #5
   #6
   #7
   #8
  ]
  properties {
   nodeName=""
   generalConfig {
    defaultMessage=The gods have harmed many. I seek to put a stop to it.
    contextualMessages [
     {
      message="Well-met, mortal. "
      contextRequirements [{requirement {class=FortControllerRequirement_IsFirstEverConversationWithNPC}}]
      priority {curve {rowName=Tandem.Speech.Priority.FirstEverInteraction}}
     }
     {
      message="You carry a great potential in you, mortal. I should know, I was much the same..."
      contextRequirements [{requirement {
         class=FortControllerRequirement_HasCID
         properties {allowedCIds [&Character_VoidRedemption_Rebel]}
        }}]
      priority {curve {rowName=Tandem.Speech.Priority.OutfitReaction}}
     }
     {
      message="Someday you gods, and all the pain you've caused, will merely be a myth."
      contextRequirements [{requirement {
         class=FortControllerRequirement_HasCID
         properties {allowedCIds [&Character_StunningMountain_Armor]}
        }}]
      priority {curve {rowName=Tandem.Speech.Priority.OutfitReaction}}
     }
     {
      message="I would love to trust you. But in the end, you're a god, just like them."
      contextRequirements [{requirement {class=FortControllerRequirement_HasCID, properties {allowedCIds [/BRCosmetics/Athena/Items/Cosmetics/Characters/Character_LiftingAura_Wave]}}}]
      priority {curve {rowName=Tandem.Speech.Priority.OutfitReaction}}
     }
     {
      message="Leave me be, warmonger. I prefer more subtle methods of combat."
      contextRequirements [{requirement {
         class=FortControllerRequirement_HasCID
         properties {allowedCIds [/BRCosmetics/Athena/Items/Cosmetics/Characters/Character_GoldAccomplishment]}
        }}]
      priority {curve {rowName=Tandem.Speech.Priority.OutfitReaction}}
     }
     {
      message=The gods have harmed many. I seek to put a stop to it.
      contextRequirements [{requirement {
         class=FortControllerRequirement_CheckNPCMatchConversationCount
         properties {nPCUniqueIdTag {tagName=AISpawnerData.Type.Tandem.ReptilianOcean}}
        }}]
      priority {curve {rowName=None}}
     }
     {
      message=I presume you've heard of me. The gods love their little myths.
      contextRequirements [{requirement {
         class=FortControllerRequirement_CheckNPCMatchConversationCount
         properties {nPCUniqueIdTag {tagName=AISpawnerData.Type.Tandem.ReptilianOcean}}
        }}]
      priority {curve {rowName=None}}
     }
     {
      message="If you ever feel like, oh, ending the rule of the pantheon? Give me a call."
      contextRequirements [{requirement {
         class=FortControllerRequirement_CheckNPCMatchConversationCount
         properties {nPCUniqueIdTag {tagName=AISpawnerData.Type.Tandem.ReptilianOcean}}
        }}]
      priority {curve {rowName=None}}
     }
     {
      message=The gods have taken much from me. But they will never take my fury.
      contextRequirements [{requirement {
         class=FortControllerRequirement_CheckNPCMatchConversationCount
         properties {nPCUniqueIdTag {tagName=AISpawnerData.Type.Tandem.ReptilianOcean}}
        }}]
      priority {curve {rowName=None}}
     }
     {
      message="If you see Zeus, ask him to return my trinkets. He knows what he stole."
      contextRequirements [{requirement {
         class=FortControllerRequirement_CheckNPCMatchConversationCount
         properties {nPCUniqueIdTag {tagName=AISpawnerData.Type.Tandem.ReptilianOcean}}
        }}]
      priority {curve {rowName=None}}
     }
     {
      message="Whatever you've heard about me, it's probably a lie."
      contextRequirements [{requirement {
         class=FortControllerRequirement_CheckNPCMatchConversationCount
         properties {nPCUniqueIdTag {tagName=AISpawnerData.Type.Tandem.ReptilianOcean}}
        }}]
      priority {curve {rowName=None}}
     }
    ]
   }
  }
 }
 ConversationLinkNode_6 {
  id=#8
  class=ConversationLinkNode
  properties {remoteEntryTag {tagName=Conversation.Entry.Season29.Snapshot.P04.Q07}}
 }
 ConversationLinkNode_5 {
  id=#7
  class=ConversationLinkNode
  properties {remoteEntryTag {tagName=Conversation.Entry.Season29.Snapshot.P04.Q03}}
 }
 ConversationLinkNode_4 {
  id=#6
  class=ConversationLinkNode
  properties {remoteEntryTag {tagName=Conversation.Entry.Season29.Snapshot.P04.Q01}}
 }
 ConversationLinkNode_3 {
  id=#5
  class=ConversationLinkNode
  properties {remoteEntryTag {tagName=Conversation.Entry.ReptilianOcean.Dialogue}}
 }
 ConversationLinkNode_0 {id=#4, class=ConversationLinkNode, properties {remoteEntryTag {tagName=Conversation.Entry.BasicServices}}}
 ConversationLinkNode_1 {id=#3, class=ConversationLinkNode, properties {remoteEntryTag {tagName=Conversation.Entry.NPCLibrary.Quests}}}
]
